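# Physics asset for the skeletal mesh of the Pickaxe_31 two-handed melee weapon cosmetic.
# It defines a single body on the Melee_TwoHanded_Handle bone, shaped as one rotated, offset box.
# It has no constraints, so the whole pickaxe collides as one rigid piece.
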PhysicsAsset /BRCosmetics/Weapons/FORT_Melee/Pickaxe_31/Meshes/SK_Pickaxe_31_PhysicsAsset
bodies {Melee_TwoHanded_Handle [{shapeType=Box, extentX=58.9879, extentY=5.7955, extentZ=101.6176, center=<0.7701 0.6498 20.4614>, rotation=<rot 4.2924 -90 -0.0118>}]}
constraints []
